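; Landmass preview actor that displays a Landmass Brush Manager's landscape as a grid of instanced static meshes with a preview material.
; SpawnInstanceMeshes stores the manager and rebuilds one instance per section, correcting scale and offset for component size.
; UpdateMID, also run on construction, feeds the manager's heightmap and named weightmap render targets into a dynamic material.
; SetMPCparams writes landscape transform, size and render-target data to MPC_Landscape.
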(event EventBeginPlay)

(fn SetMPCparams
  (bind _manager (Variables|Default|GetManager))
  (bind _landscapeinformation (Variables|/Script/LandmassEditor.LandmassManagerBase|GetLandscapeInformation _manager))
  (bind (_landscapetransform _landscapequads _rendertargetresolution) (Utilities|Struct|BreakLandmassLandscapeInfo _landscapeinformation))
  (bind (_x _y) (Utilities|Struct|BreakIntPoint _rendertargetresolution))
  (bind (_x_2 _y_2) (Utilities|Struct|BreakIntPoint _landscapequads))
  (bind _manager_2 (Variables|Default|GetManager))
  (bind _landscapeinformation_2 (Variables|/Script/LandmassEditor.LandmassManagerBase|GetLandscapeInformation _manager_2))
  (bind _landscapetransform_2 (Utilities|Struct|BreakLandmassLandscapeInfo _landscapeinformation_2))
  (bind _manager_3 (Variables|Default|GetManager))
  (bind _landscapeinformation_3 (Variables|/Script/LandmassEditor.LandmassManagerBase|GetLandscapeInformation _manager_3))
  (bind _landscapetransform_3 (Utilities|Struct|BreakLandmassLandscapeInfo _landscapeinformation_3))
  (bind (_x_3 _y_3) (Math|Vector2D|BreakVector2D (Variables|/Landmass/Landscape/BlueprintBrushes/LandmassBrushManager.LandmassBrushManager_C|GetRT%20World%20Size (Variables|Default|GetManager))))
  (bind (_x_4 _y_4) (Math|Vector2D|BreakVector2D (Variables|/Landmass/Landscape/BlueprintBrushes/LandmassBrushManager.LandmassBrushManager_C|GetRT%20World%20Center (Variables|Default|GetManager))))
  (bind _world_size (Variables|/Landmass/Landscape/BlueprintBrushes/LandmassBrushManager.LandmassBrushManager_C|GetWorld%20Size _manager_3))
  (bind _manager_4 (Variables|Default|GetManager))
  (Rendering|Material|SetScalarParameterValue "/Landmass/Landscape/BlueprintBrushes/MPC/MPC_Landscape.MPC_Landscape" "RTResX" (Math|Conversions|ToFloat(Integer) _x) :!class /Script/Engine.KismetMaterialLibrary)
  (Rendering|Material|SetScalarParameterValue "/Landmass/Landscape/BlueprintBrushes/MPC/MPC_Landscape.MPC_Landscape" "RTResY" (Math|Conversions|ToFloat(Integer) _y) :!class /Script/Engine.KismetMaterialLibrary)
  (Rendering|Material|SetScalarParameterValue "/Landmass/Landscape/BlueprintBrushes/MPC/MPC_Landscape.MPC_Landscape" "LSQuadsX" (Math|Conversions|ToFloat(Integer) _x_2) :!class /Script/Engine.KismetMaterialLibrary)
  (Rendering|Material|SetScalarParameterValue "/Landmass/Landscape/BlueprintBrushes/MPC/MPC_Landscape.MPC_Landscape" "LSQuadsY" (Math|Conversions|ToFloat(Integer) _y_2) :!class /Script/Engine.KismetMaterialLibrary)
  (Rendering|Material|SetScalarParameterValue "/Landmass/Landscape/BlueprintBrushes/MPC/MPC_Landscape.MPC_Landscape" "WorldSizeX" (.x _world_size) :!class /Script/Engine.KismetMaterialLibrary)
  (Rendering|Material|SetScalarParameterValue "/Landmass/Landscape/BlueprintBrushes/MPC/MPC_Landscape.MPC_Landscape" "WorldSizeY" (.y _world_size) :!class /Script/Engine.KismetMaterialLibrary)
  (Rendering|Material|SetScalarParameterValue "/Landmass/Landscape/BlueprintBrushes/MPC/MPC_Landscape.MPC_Landscape" "LandscapeZLocation" (.z (.location _landscapetransform_3)) :!class /Script/Engine.KismetMaterialLibrary)
  (Rendering|Material|SetScalarParameterValue "/Landmass/Landscape/BlueprintBrushes/MPC/MPC_Landscape.MPC_Landscape" "LandscapeZScale" (/ (.z (.scale _landscapetransform_3)) 128.0) :!class /Script/Engine.KismetMaterialLibrary)
  (Rendering|Material|SetVectorParameterValue "/Landmass/Landscape/BlueprintBrushes/MPC/MPC_Landscape.MPC_Landscape" "LandscapeLocation" (Math|Conversions|ToLinearColor(Vector) (.location _landscapetransform_2)) :!class /Script/Engine.KismetMaterialLibrary)
  (Rendering|Material|SetVectorParameterValue "/Landmass/Landscape/BlueprintBrushes/MPC/MPC_Landscape.MPC_Landscape" "LandscapeScale" (Math|Conversions|ToLinearColor(Vector) (.scale _landscapetransform_2)) :!class /Script/Engine.KismetMaterialLibrary)
  (Rendering|Material|SetVectorParameterValue "/Landmass/Landscape/BlueprintBrushes/MPC/MPC_Landscape.MPC_Landscape" "RTWorldLocation" (Math|Color|MakeColor _x_4 _y_4) :!class /Script/Engine.KismetMaterialLibrary)
  (Rendering|Material|SetVectorParameterValue "/Landmass/Landscape/BlueprintBrushes/MPC/MPC_Landscape.MPC_Landscape" "RTWorldSize" (Math|Color|MakeColor _x_3 _y_3) :!class /Script/Engine.KismetMaterialLibrary)
  (Rendering|Material|SetScalarParameterValue "/Landmass/Landscape/BlueprintBrushes/MPC/MPC_Landscape.MPC_Landscape" "ComponentRes" 256.0 :!class /Script/Engine.KismetMaterialLibrary))

(fn SpawnInstanceMeshes
  (params (Brush%20Manager :type Landmass%20Brush%20Manager%20Object%20Reference))
  (bind _manager (Variables|Default|GetManager))
  (bind _landscapeinformation (Variables|/Script/LandmassEditor.LandmassManagerBase|GetLandscapeInformation _manager))
  (bind (_landscapetransform _landscapequads) (Utilities|Struct|BreakLandmassLandscapeInfo _landscapeinformation))
  (bind _x (Utilities|Struct|BreakIntPoint _landscapequads))
  (bind _returnvalue (Math|Conversions|ToFloat(Integer) _x))
  (bind (_x_2 _y) (Math|Vector2D|BreakVector2D (Math|Conversions|ToVector2D(Vector) (Variables|/Landmass/Landscape/BlueprintBrushes/LandmassBrushManager.LandmassBrushManager_C|GetWorld%20Size (Variables|Default|GetManager)))))
  (bind (_x_3 _y_2) (Math|Vector2D|BreakVector2D (Math|Conversions|ToVector2D(Vector) (Variables|/Landmass/Landscape/BlueprintBrushes/LandmassBrushManager.LandmassBrushManager_C|GetWorld%20Size (Variables|Default|GetManager)))))
  (bind (_origin _boxextent) (Math|BoxSphereBounds|BreakBoxSphereBounds (StaticMesh|GetBounds (Variables|/Script/Engine.StaticMeshComponent|GetStaticMesh (Variables|Default|GetLandscape%20ISMC)))))
  (bind _manager_2 (Variables|Default|GetManager))
  (bind _sections_x (Variables|/Landmass/Landscape/BlueprintBrushes/LandmassBrushManager.LandmassBrushManager_C|GetSections%20X _manager_2))
  (bind _returnvalue_2 (* _sections_x 255))
  (Variables|Default|SetManager Brush%20Manager)
  (Components|InstancedStaticMesh|ClearInstances (Variables|Default|GetLandscape%20ISMC))
  (Utilities|FlowControl|Sequence
    (:then_0
      (LocalVariables|SetCorrectForComponentSize (/ _returnvalue _returnvalue_2))
      (LocalVariables|SetCorrectForSize_Offset (* (/ (- _returnvalue_2 _returnvalue) 2.0) (.x (.scale (Variables|/Landmass/Landscape/BlueprintBrushes/LandmassBrushManager.LandmassBrushManager_C|GetLS%20Transform (Variables|Default|GetManager))))))
      ;[1 1 1] Add Instances
      (bind (_current_point _total_points _0_1_phase _poly_center) (Utilities|2DGridExecutionMacro :Sector%20Size (/ (/ (Math|Float|Max(Float) _x_2 _y) (Math|Float|Max(Float) (Math|Conversions|ToFloat(Integer) (Variables|/Landmass/Landscape/BlueprintBrushes/LandmassBrushManager.LandmassBrushManager_C|GetSections%20X (Variables|Default|GetManager))) (Math|Conversions|ToFloat(Integer) (Variables|/Landmass/Landscape/BlueprintBrushes/LandmassBrushManager.LandmassBrushManager_C|GetSections%20Y (Variables|Default|GetManager))))) (LocalVariables|GetCorrectForComponentSize)) :X (Variables|/Landmass/Landscape/BlueprintBrushes/LandmassBrushManager.LandmassBrushManager_C|GetSections%20X (Variables|Default|GetManager)) :Y (Variables|/Landmass/Landscape/BlueprintBrushes/LandmassBrushManager.LandmassBrushManager_C|GetSections%20Y (Variables|Default|GetManager)))
        (:Per%20Poly%20Execution
          (Components|InstancedStaticMesh|AddInstance (Variables|Default|GetLandscape%20ISMC) (Math|Transform|MakeTransform (+ (+ (* _poly_center "1.000000,1.000000, 0") (/ (+ (.location (Variables|/Landmass/Landscape/BlueprintBrushes/LandmassBrushManager.LandmassBrushManager_C|GetLS%20Transform (Variables|Default|GetManager))) (+ (.location (Variables|/Landmass/Landscape/BlueprintBrushes/LandmassBrushManager.LandmassBrushManager_C|GetLS%20Transform (Variables|Default|GetManager))) (Variables|/Landmass/Landscape/BlueprintBrushes/LandmassBrushManager.LandmassBrushManager_C|GetWorld%20Size (Variables|Default|GetManager)))) 2.0)) (Math|Vector|MakeVector (LocalVariables|GetCorrectForSize_Offset) (LocalVariables|GetCorrectForSize_Offset))) "0, 0, 0" (Math|Vector|MakeVector (/ (/ (/ _x_3 (Math|Conversions|ToFloat(Integer) (Variables|/Landmass/Landscape/BlueprintBrushes/LandmassBrushManager.LandmassBrushManager_C|GetSections%20X (Variables|Default|GetManager)))) (* (.x _boxextent) 2.0)) (LocalVariables|GetCorrectForComponentSize)) (/ (/ (/ _y_2 (Math|Conversions|ToFloat(Integer) (Variables|/Landmass/Landscape/BlueprintBrushes/LandmassBrushManager.LandmassBrushManager_C|GetSections%20Y (Variables|Default|GetManager)))) (* (.y _boxextent) 2.0)) (LocalVariables|GetCorrectForComponentSize)) 1.0)) true))))
    (:then_1
      (CallFunction|UpdateMID))))

(fn UpdateMID
  (Rendering|CreateAndAssignMID (Variables|Default|GetPreview%20Landscape%20MID) (Variables|Default|GetPreview%20Material)
    (:Out
      (Rendering|Material|SetMaterial (Variables|Default|GetLandscape%20ISMC) 0 (Variables|Default|GetPreview%20Landscape%20MID))
      (CallFunction|UpdateMIDHeight (Variables|/Landmass/Landscape/BlueprintBrushes/LandmassBrushManager.LandmassBrushManager_C|GetHeightmap%20Render%20RT (Variables|Default|GetManager)))
      (bind _keys (Utilities|Map|Keys (Variables|/Landmass/Landscape/BlueprintBrushes/LandmassBrushManager.LandmassBrushManager_C|GetWeightmap%20RTs (Variables|Default|GetManager))))
      (for _array_element _keys
        (CallFunction|UpdateMIDWeight _array_element (Utilities|Map|Find (Variables|/Landmass/Landscape/BlueprintBrushes/LandmassBrushManager.LandmassBrushManager_C|GetWeightmap%20RTs (Variables|Default|GetManager)) _array_element))))))

(fn UpdateMIDHeight
  (params (Height%20RT :type Texture%20Object%20Reference))
  (Rendering|Material|SetTextureParameterValue (Variables|Default|GetPreview%20Landscape%20MID) "HeightRT" Height%20RT))

(fn UpdateMIDWeight
  (params (Name :type Name) (Weightmap%20RT :type Texture%20Object%20Reference))
  (Rendering|Material|SetTextureParameterValue (Variables|Default|GetPreview%20Landscape%20MID) Name Weightmap%20RT))

(fn ConstructionScript
  (Utilities|IsValid (Variables|Default|GetManager)
    (:Is%20Valid
      (CallFunction|UpdateMID))))
